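# Physics asset for the male Commando GumshoeDark backpack mesh, a Battle Royale back-slot cosmetic accessory.
# It holds a single capsule body on the backpack_root bone, offset and rotated from the bone's origin.
# It defines no constraints, so the backpack is represented as one simple collision volume with no jointed parts.
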PhysicsAsset /BRCosmetics/Accessories/FORT_Backpacks/M_MED_GumshoeDark_Backpack/Meshes/Male_Commando_GumshoeDark_PhysicsAsset
bodies {backpack_root [{shapeType=Capsule, length=19.1922, radius=18.5389, center=<-0.0818 -8.0204 -2.0449>, rotation=<rot 4.9167 -90 39.5575>}]}
constraints []
